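# Translucent unlit mesh-particle material for a spooky mist camera lens effect, coloured between dark and particle-tinted light tones.
# Two counter-panning samples of a storm swirl mask form the mist, one varying the other's power exponent, with a soft vertical sphere mask shaping it.
# Particle dynamic parameters scale pan speed and power, while CloseUp blends in a V-weighted offset toward the object's position.
# Opacity follows particle alpha and uses depth fade except on mobile.
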Material /Figment_S06/Effects/Fort_Effects/Materials/LensEffect/M_CameraLens_SpookyMist_DepthCheck
graph {
 properties {bUsedWithMeshParticles=true, blendMode=BLEND_Translucent, pixelDepthOffsetMode=PDOM_Legacy, shadingModel=MSM_Unlit, usedShadingModels=Unlit}
 nodes [
  n1 = TextureCoordinate()
  n2 = ObjectPositionWS()
  n3 = WorldPosition()
  LinearInterpolate(
   B=Multiply(A=ComponentMask(None=@n1) {g=true}, B=Subtract(A=@n2, B=@n3.XYZ))
   Alpha=DynamicParameter() {defaultValue=<rgba 0 1 1 1>, paramNames [CloseUp,PowerMult,PanSpeed,Param4]}.CloseUp
  ) {constAlpha=1}
  n4 = ParticleColor()
  n5 = Multiply(A=TextureCoordinate(), B=ComponentMask(None=VectorParameter() {parameterName=Tiling, defaultValue=<rgba 3 0.2 0 0>}.RGB) {g=true, r=true})
  n6 = DynamicParameter() {defaultValue=<rgba 0 1 1 1>, paramNames [CloseUp,PowerMult,PanSpeed,Param4]}
  n7 = Multiply(A=Multiply(A=@n6.PanSpeed, B=Time()), B=ScalarParameter() {defaultValue=3, parameterName=PanSpeed})
  n8 = ScalarParameter() {defaultValue=7, parameterName=Power}
  n9 = Panner(Coordinate=TextureCoordinate() {uTiling=2, vTiling=0.1}, Time=@n7) {speedX=0.015, speedY=-0.02}
  n10 = Power(
   Base=TextureSample(UVs=Panner(Coordinate=@n5, Time=@n7) {speedX=-0.01, speedY=0.035}) {samplerType=SAMPLERTYPE_LinearGrayscale, texture=/Figment_S06/VisualThreatManager/StormVisuals/LargeClouds/T_StormSwirlMask}.G
   Exp=LinearInterpolate(
    A=Multiply(A=@n8, B=@n6.PowerMult)
    Alpha=TextureSample(UVs=@n9) {samplerType=SAMPLERTYPE_LinearGrayscale, texture=&T_StormSwirlMask}.G
   ) {constA=7, constB=0.6}
  )
  n11 = TextureCoordinate()
  n12 = SphereMask(A=OneMinus(None=ComponentMask(None=@n11) {g=true}), B=Constant() {r=0.5}) {attenuationRadius=0.5, hardnessPercent=0}
  n13 = Power(Base=@n12)
  LinearInterpolate(
   A=VectorParameter() {parameterName=DarkColor, defaultValue=<rgba 0.0055 0.0075 0.01 0>}.RGB
   B=Multiply(A=@n4.RGB, B=VectorParameter() {parameterName=LightColor, defaultValue=<rgba 0.233 0.1454 0.6 0>}.RGB)
   Alpha=Saturate(None=Multiply(A=@n10, B=@n13))
  )
  n14 = Multiply(A=Multiply(A=@n4.A, B=@n10), B=@n12)
  ShadingPathSwitch(Default=DepthFade(Opacity=@n14), Mobile=@n14) {desc=Migrated from FeatureLevelSwitch}
 ]
}
